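# Physics asset for the female medium HornedMask body cosmetic, covering the skirt bones and both thighs.
# Each side has five skirt strips (front, frontMid, side, sideMid, rear), each a three-box chain rooted on a tiny master capsule.
# Strip joints share tight twist and first-swing limits with a wider second swing.
# Each master joins its thigh capsule with free swing and twist, leaving the left and right halves unconstrained to each other.
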PhysicsAsset /BRCosmetics/Characters/Player/Female/Medium/Bodies/F_MED_HornedMask/Meshes/F_MED_HornedMask_Body_Physics
bodies {
 dyn_skirt_frontMid_01_l [{shapeType=Box, extentX=10, extentY=10, extentZ=3, center=<5 0 0>}]
 dyn_skirt_frontMid_01_r [{shapeType=Box, extentX=10, extentY=10, extentZ=3, center=<5 0 0>}]
 dyn_skirt_frontMid_02_l [{shapeType=Box, extentX=10, extentY=10, extentZ=3, center=<5 0 0>}]
 dyn_skirt_frontMid_02_r [{shapeType=Box, extentX=10, extentY=10, extentZ=3, center=<5 0 0>}]
 dyn_skirt_frontMid_03_l [{shapeType=Box, extentX=10, extentY=10, extentZ=3, center=<5 0 0>}]
 dyn_skirt_frontMid_03_r [{shapeType=Box, extentX=10, extentY=10, extentZ=3, center=<5 0 0>}]
 dyn_skirt_front_01_l [{shapeType=Box, extentX=10, extentY=10, extentZ=3, center=<5 0 0>}]
 dyn_skirt_front_01_r [{shapeType=Box, extentX=10, extentY=10, extentZ=3, center=<5 0 0>}]
 dyn_skirt_front_02_l [{shapeType=Box, extentX=10, extentY=10, extentZ=3, center=<5 0 0>}]
 dyn_skirt_front_02_r [{shapeType=Box, extentX=10, extentY=10, extentZ=3, center=<5 0 0>}]
 dyn_skirt_front_03_l [{shapeType=Box, extentX=10, extentY=10, extentZ=3, center=<5 0 0>}]
 dyn_skirt_front_03_r [{shapeType=Box, extentX=10, extentY=10, extentZ=3, center=<5 0 0>}]
 dyn_skirt_master_l [{shapeType=Capsule, length=0.505, radius=0.505, rotation=<rot 0 -90 0>}]
 dyn_skirt_master_r [{shapeType=Capsule, length=0.505, radius=0.505, rotation=<rot 0 -90 0>}]
 dyn_skirt_rear_01_l [{shapeType=Box, extentX=10, extentY=10, extentZ=3, center=<5 0 0>}]
 dyn_skirt_rear_01_r [{shapeType=Box, extentX=10, extentY=10, extentZ=3, center=<5 0 0>}]
 dyn_skirt_rear_02_l [{shapeType=Box, extentX=10, extentY=10, extentZ=3, center=<5 0 0>}]
 dyn_skirt_rear_02_r [{shapeType=Box, extentX=10, extentY=10, extentZ=3, center=<5 0 0>}]
 dyn_skirt_rear_03_l [{shapeType=Box, extentX=10, extentY=10, extentZ=3, center=<5 0 0>}]
 dyn_skirt_rear_03_r [{shapeType=Box, extentX=10, extentY=10, extentZ=3, center=<5 0 0>}]
 dyn_skirt_sideMid_01_l [{shapeType=Box, extentX=10, extentY=10, extentZ=3, center=<5 0 0>}]
 dyn_skirt_sideMid_01_r [{shapeType=Box, extentX=10, extentY=10, extentZ=3, center=<5 0 0>}]
 dyn_skirt_sideMid_02_l [{shapeType=Box, extentX=10, extentY=10, extentZ=3, center=<5 0 0>}]
 dyn_skirt_sideMid_02_r [{shapeType=Box, extentX=10, extentY=10, extentZ=3, center=<5 0 0>}]
 dyn_skirt_sideMid_03_l [{shapeType=Box, extentX=10, extentY=10, extentZ=3, center=<5 0 0>}]
 dyn_skirt_sideMid_03_r [{shapeType=Box, extentX=10, extentY=10, extentZ=3, center=<5 0 0>}]
 dyn_skirt_side_01_l [{shapeType=Box, extentX=10, extentY=10, extentZ=3, center=<5 0 0>}]
 dyn_skirt_side_01_r [{shapeType=Box, extentX=10, extentY=10, extentZ=3, center=<5 0 0>}]
 dyn_skirt_side_02_l [{shapeType=Box, extentX=10, extentY=10, extentZ=3, center=<5 0 0>}]
 dyn_skirt_side_02_r [{shapeType=Box, extentX=10, extentY=10, extentZ=3, center=<5 0 0>}]
 dyn_skirt_side_03_l [{shapeType=Box, extentX=10, extentY=10, extentZ=3, center=<5 0 0>}]
 dyn_skirt_side_03_r [{shapeType=Box, extentX=10, extentY=10, extentZ=3, center=<5 0 0>}]
 thigh_l [{shapeType=Capsule, length=30, radius=7.9731, center=<-20 0 0>, rotation=<rot 90 0 0>}]
 thigh_r [{shapeType=Capsule, length=30, radius=7.9731, center=<20 0 0>, rotation=<rot 90 0 0>}]
}
constraints [
 {bone1Name=dyn_skirt_rear_02_r, bone2Name=dyn_skirt_rear_01_r, swing1LimitDegrees=10, swing2LimitDegrees=45, twistLimitDegrees=5}
 {bone1Name=dyn_skirt_rear_03_r, bone2Name=dyn_skirt_rear_02_r, swing1LimitDegrees=10, swing2LimitDegrees=45, twistLimitDegrees=5}
 {bone1Name=dyn_skirt_side_02_r, bone2Name=dyn_skirt_side_01_r, swing1LimitDegrees=10, swing2LimitDegrees=45, twistLimitDegrees=5}
 {bone1Name=dyn_skirt_side_03_r, bone2Name=dyn_skirt_side_02_r, swing1LimitDegrees=10, swing2LimitDegrees=45, twistLimitDegrees=5}
 {bone1Name=dyn_skirt_front_02_r, bone2Name=dyn_skirt_front_01_r, swing1LimitDegrees=10, swing2LimitDegrees=45, twistLimitDegrees=5}
 {bone1Name=dyn_skirt_front_03_r, bone2Name=dyn_skirt_front_02_r, swing1LimitDegrees=10, swing2LimitDegrees=45, twistLimitDegrees=5}
 {bone1Name=dyn_skirt_sideMid_02_r, bone2Name=dyn_skirt_sideMid_01_r, swing1LimitDegrees=10, swing2LimitDegrees=45, twistLimitDegrees=5}
 {bone1Name=dyn_skirt_sideMid_03_r, bone2Name=dyn_skirt_sideMid_02_r, swing1LimitDegrees=10, swing2LimitDegrees=45, twistLimitDegrees=5}
 {bone1Name=dyn_skirt_frontMid_02_r, bone2Name=dyn_skirt_frontMid_01_r, swing1LimitDegrees=10, swing2LimitDegrees=45, twistLimitDegrees=5}
 {bone1Name=dyn_skirt_frontMid_03_r, bone2Name=dyn_skirt_frontMid_02_r, swing1LimitDegrees=10, swing2LimitDegrees=45, twistLimitDegrees=5}
 {bone1Name=dyn_skirt_rear_02_l, bone2Name=dyn_skirt_rear_01_l, swing1LimitDegrees=10, swing2LimitDegrees=45, twistLimitDegrees=5}
 {bone1Name=dyn_skirt_rear_03_l, bone2Name=dyn_skirt_rear_02_l, swing1LimitDegrees=10, swing2LimitDegrees=45, twistLimitDegrees=5}
 {bone1Name=dyn_skirt_side_02_l, bone2Name=dyn_skirt_side_01_l, swing1LimitDegrees=10, swing2LimitDegrees=45, twistLimitDegrees=5}
 {bone1Name=dyn_skirt_side_03_l, bone2Name=dyn_skirt_side_02_l, swing1LimitDegrees=10, swing2LimitDegrees=45, twistLimitDegrees=5}
 {bone1Name=dyn_skirt_front_02_l, bone2Name=dyn_skirt_front_01_l, swing1LimitDegrees=10, swing2LimitDegrees=45, twistLimitDegrees=5}
 {bone1Name=dyn_skirt_front_03_l, bone2Name=dyn_skirt_front_02_l, swing1LimitDegrees=10, swing2LimitDegrees=45, twistLimitDegrees=5}
 {bone1Name=dyn_skirt_sideMid_02_l, bone2Name=dyn_skirt_sideMid_01_l, swing1LimitDegrees=10, swing2LimitDegrees=45, twistLimitDegrees=5}
 {bone1Name=dyn_skirt_sideMid_03_l, bone2Name=dyn_skirt_sideMid_02_l, swing1LimitDegrees=10, swing2LimitDegrees=45, twistLimitDegrees=5}
 {bone1Name=dyn_skirt_frontMid_03_l, bone2Name=dyn_skirt_frontMid_02_l, swing1LimitDegrees=10, swing2LimitDegrees=45, twistLimitDegrees=5}
 {bone1Name=dyn_skirt_frontMid_02_l, bone2Name=dyn_skirt_frontMid_01_l, swing1LimitDegrees=10, swing2LimitDegrees=45, twistLimitDegrees=5}
 {bone1Name=dyn_skirt_rear_01_l, bone2Name=dyn_skirt_master_l, swing1LimitDegrees=10, swing2LimitDegrees=45, twistLimitDegrees=5}
 {bone1Name=dyn_skirt_side_01_l, bone2Name=dyn_skirt_master_l, swing1LimitDegrees=10, swing2LimitDegrees=45, twistLimitDegrees=5}
 {bone1Name=dyn_skirt_front_01_l, bone2Name=dyn_skirt_master_l, swing1LimitDegrees=10, swing2LimitDegrees=45, twistLimitDegrees=5}
 {bone1Name=dyn_skirt_sideMid_01_l, bone2Name=dyn_skirt_master_l, swing1LimitDegrees=10, swing2LimitDegrees=45, twistLimitDegrees=5}
 {bone1Name=dyn_skirt_frontMid_01_l, bone2Name=dyn_skirt_master_l, swing1LimitDegrees=10, swing2LimitDegrees=45, twistLimitDegrees=5}
 {bone1Name=dyn_skirt_rear_01_r, bone2Name=dyn_skirt_master_r, swing1LimitDegrees=10, swing2LimitDegrees=45, twistLimitDegrees=5}
 {bone1Name=dyn_skirt_side_01_r, bone2Name=dyn_skirt_master_r, swing1LimitDegrees=10, swing2LimitDegrees=45, twistLimitDegrees=5}
 {bone1Name=dyn_skirt_front_01_r, bone2Name=dyn_skirt_master_r, swing1LimitDegrees=10, swing2LimitDegrees=45, twistLimitDegrees=5}
 {bone1Name=dyn_skirt_sideMid_01_r, bone2Name=dyn_skirt_master_r, swing1LimitDegrees=10, swing2LimitDegrees=45, twistLimitDegrees=5}
 {bone1Name=dyn_skirt_frontMid_01_r, bone2Name=dyn_skirt_master_r, swing1LimitDegrees=10, swing2LimitDegrees=45, twistLimitDegrees=5}
 {
  bone1Name=dyn_skirt_master_l
  bone2Name=thigh_l
  swing1LimitDegrees=45
  swing1Motion=Free
  swing2LimitDegrees=45
  swing2Motion=Free
  twistLimitDegrees=45
  twistMotion=Free
 }
 {
  bone1Name=dyn_skirt_master_r
  bone2Name=thigh_r
  swing1LimitDegrees=45
  swing1Motion=Free
  swing2LimitDegrees=45
  swing2Motion=Free
  twistLimitDegrees=45
  twistMotion=Free
 }
]
