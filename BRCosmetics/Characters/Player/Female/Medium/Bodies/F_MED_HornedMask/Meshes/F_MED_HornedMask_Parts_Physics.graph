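# Physics asset for the Horned Mask medium female player cosmetic's parts mesh, covering its dangling waist tassels and ribbons.
# Four tassel capsules are constrained to spine_01 with angular limits, and two ribbon capsules to the pelvis with twist locked.
# Pelvis and spine_01 are tiny anchor capsules, freely rotating in their constraints like the thighs.
# Long thigh capsules supply leg volume, with the right thigh carrying a second, differently angled capsule.
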PhysicsAsset /BRCosmetics/Characters/Player/Female/Medium/Bodies/F_MED_HornedMask/Meshes/F_MED_HornedMask_Parts_Physics
bodies {
 dyn_waistTassel_front_l [{shapeType=Capsule, length=4.6461, radius=4.9427, center=<5.2853 -0.5325 6.2254>, rotation=<rot -4.929 -90 43.7233>}]
 dyn_waistTassel_front_r [{shapeType=Capsule, length=8.8461, radius=4.9427, center=<3.0254 0.3295 3.8713>, rotation=<rot 4.929 -90 43.7233>}]
 dyn_waistTassel_rear_l [{shapeType=Capsule, length=8.6504, radius=4.9104, center=<2.8642 0.1343 3.8971>, rotation=<rot 0.3567 -90 37.8721>}]
 dyn_waistTassel_rear_r [{shapeType=Capsule, length=8.6504, radius=4.9104, center=<2.8646 -0.1343 3.8971>, rotation=<rot -0.3567 -90 37.8721>}]
 dyn_waist_ribbon_l [{shapeType=Capsule, length=6.8665, radius=4.0277, center=<5.0804 0 0>, rotation=<rot 0 90 90>}]
 dyn_waist_ribbon_r [{shapeType=Capsule, length=7.1934, radius=4.3363, center=<5 0 0>, rotation=<rot 0 90 90>}]
 pelvis [{shapeType=Capsule, length=1, radius=1}]
 spine_01 [{shapeType=Capsule, length=1, radius=1}]
 thigh_l [{
   shapeType=Capsule
   length=36.4772
   radius=8.4721
   shapeName=thigh_l_capsule
   center=<-19.7859 -2.9159 0.295>
   rotation=<rot 0 95.5543 -89.6477>
  }]
 thigh_r [
  {
   shapeType=Capsule
   length=36.4772
   radius=8.4721
   shapeName=thigh_r_capsule
   center=<19.9924 2.9159 -0.2999>
   rotation=<rot 0 95.5543 -89.6477>
  }
  {
   shapeType=Capsule
   length=36.4772
   radius=8.4721
   shapeName=thigh_r_capsule
   center=<20.7103 -0.6184 -0.2976>
   rotation=<rot 0.0687 84.3043 -89.6545>
  }
 ]
}
constraints [
 {
  bone1Name=spine_01
  bone2Name=pelvis
  swing1LimitDegrees=45
  swing1Motion=Free
  swing2LimitDegrees=45
  swing2Motion=Free
  twistLimitDegrees=45
  twistMotion=Free
 }
 {bone1Name=dyn_waistTassel_rear_r, bone2Name=spine_01, swing1LimitDegrees=10, swing2LimitDegrees=30, twistLimitDegrees=20}
 {bone1Name=dyn_waistTassel_front_r, bone2Name=spine_01, swing1LimitDegrees=10, swing2LimitDegrees=30, twistLimitDegrees=20}
 {bone1Name=dyn_waistTassel_front_l, bone2Name=spine_01, swing1LimitDegrees=10, swing2LimitDegrees=30, twistLimitDegrees=20}
 {bone1Name=dyn_waistTassel_rear_l, bone2Name=spine_01, swing1LimitDegrees=10, swing2LimitDegrees=30, twistLimitDegrees=20}
 {bone1Name=dyn_waist_ribbon_l, bone2Name=pelvis, swing1LimitDegrees=45, swing2LimitDegrees=20, twistLimitDegrees=10, twistMotion=Locked}
 {bone1Name=dyn_waist_ribbon_r, bone2Name=pelvis, swing1LimitDegrees=45, swing2LimitDegrees=20, twistLimitDegrees=10, twistMotion=Locked}
 {
  bone1Name=thigh_l
  bone2Name=pelvis
  swing1LimitDegrees=45
  swing1Motion=Free
  swing2LimitDegrees=45
  swing2Motion=Free
  twistLimitDegrees=45
  twistMotion=Free
 }
 {
  bone1Name=thigh_r
  bone2Name=pelvis
  swing1LimitDegrees=45
  swing1Motion=Free
  swing2LimitDegrees=45
  swing2Motion=Free
  twistLimitDegrees=45
  twistMotion=Free
 }
]
