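# Physics asset of cape colliders for the F_MED_HornedMask female character body, named as the collision set that keeps the cape off the body.
# Capsules cover the head, clavicles, pelvis, spine chain and feet, with crossed capsule pairs on the middle spine bones.
# Extra capsules ride on dynamic rear skirt, waist tassel and ribbon bones.
# Arm, leg and neck bodies have no shapes, and no constraints exist, so it defines collision only.
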PhysicsAsset /BRCosmetics/Characters/Player/Female/Medium/Bodies/F_MED_HornedMask/Meshes/F_MED_HornedMask_CapeColliders
bodies {
 calf_l []
 calf_r []
 clavicle_l [{shapeType=Capsule, length=4.6964, radius=10.6353, center=<13.8621 -3.029 -3.4169>, rotation=<rot -77.5566 177.3915 7.7314>}]
 clavicle_r [{shapeType=Capsule, length=7.6964, radius=10.5457, center=<-13.1915 4.8491 3.8051>, rotation=<rot -77.5565 177.3915 -172.2685>}]
 dyn_skirt_rear_03_l [{
   shapeType=Capsule
   length=19.5
   radius=16
   shapeName=dyn_skirt_rear_03_l_capsule
   center=<-0.1727 -2.7486 8.8895>
   rotation=<rot 80.0001 0 0>
  }]
 dyn_skirt_rear_03_r [{shapeType=Capsule, length=19.5, radius=16, shapeName=dyn_skirt_rear_03_r_capsule, center=<-0.1727 3.0101 8.8895>, rotation=<rot 80.0001 0 0>}]
 dyn_waistTassel_rear_l [{
   shapeType=Capsule
   length=3.75
   radius=6.5
   shapeName=dyn_waistTassel_rear_l_capsule
   center=<5.4542 0 6.1201>
   rotation=<rot -30.0001 0.0001 0>
  }]
 dyn_waistTassel_rear_r [{
   shapeType=Capsule
   length=3.75
   radius=6.5
   shapeName=dyn_waistTassel_rear_r_capsule
   center=<5.4542 0 6.1201>
   rotation=<rot -30.0001 -0.0001 0>
  }]
 dyn_waist_ribbon_l [{
   shapeType=Capsule
   length=6.25
   radius=8.25
   shapeName=dyn_waist_ribbon_l_capsule
   center=<11.4236 -4.2949 3.688>
   rotation=<rot 90 90 109.9995>
  }]
 dyn_waist_ribbon_r [{
   shapeType=Capsule
   length=6.25
   radius=8.25
   shapeName=dyn_waist_ribbon_r_capsule
   center=<11.4236 -4.2949 -3.8932>
   rotation=<rot 90 180 -160.0004>
  }]
 foot_l [{shapeType=Capsule, length=12.5982, radius=6.5127, center=<-2.2072 6.0534 -0.3367>, rotation=<rot 0.0388 1.4701 -89.3982>}]
 foot_r [{shapeType=Capsule, length=12.5982, radius=6.5127, center=<2.2072 -6.0534 0.3367>, rotation=<rot 0.0388 1.4701 90.6018>}]
 head [{shapeType=Capsule, length=0, radius=18, center=<9.0697 0.0746 0>, rotation=<rot 0 -90 0>}]
 lowerarm_l []
 lowerarm_r []
 neck_02 []
 pelvis [{shapeType=Capsule, length=9.4028, radius=14.9096, center=<-0.1908 -4.8912 -0.1981>}]
 spine_01 [{shapeType=Capsule, length=16.5171, radius=9.9854, center=<0.2006 -3.2358 0>}]
 spine_02 [
  {shapeType=Capsule, length=3.2984, radius=8.4854, center=<1.179 2.2974 0>}
  {shapeType=Capsule, length=20.6875, radius=9.3125, center=<-1.8317 2.606 0>, rotation=<rot 90 0 0>}
 ]
 spine_03 [
  {shapeType=Capsule, length=3.2984, radius=13.2354, center=<-0.0747 1.6924 0>}
  {shapeType=Capsule, length=20.7188, radius=9.2812, center=<-1.2326 2.6059 0>, rotation=<rot 90 0 0>}
 ]
 spine_04 [
  {shapeType=Capsule, length=5.8296, radius=10.3916, center=<5.5182 1.4112 0>}
  {shapeType=Capsule, length=10.0312, radius=9.2812, center=<-1.2326 1.308 0>, rotation=<rot 90 0 0>}
 ]
 spine_05 [{shapeType=Capsule, length=10.9859, radius=10.0791, center=<5.6377 0.3227 0>}]
 thigh_l []
 thigh_r []
 upperarm_l []
 upperarm_r []
}
constraints []
