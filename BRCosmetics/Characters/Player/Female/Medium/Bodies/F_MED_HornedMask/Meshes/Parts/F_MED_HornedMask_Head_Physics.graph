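# Physics asset for the head part of the HornedMask medium female player cosmetic, defining bodies for the head, two ponytail bones and left and right ribbons.
# Each bone has a single capsule, and both ponytails and both ribbons are constrained directly to the head rather than chained together.
# The ponytail pair and the ribbon pair carry different swing and twist limit angles, and only the ribbon constraints explicitly lock twist.
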PhysicsAsset /BRCosmetics/Characters/Player/Female/Medium/Bodies/F_MED_HornedMask/Meshes/Parts/F_MED_HornedMask_Head_Physics
bodies {
 dyn_ribbon_l [{shapeType=Capsule, length=2, radius=2, center=<0 0 2>}]
 dyn_ribbon_r [{shapeType=Capsule, length=2, radius=2, center=<0 0 1>}]
 head [{shapeType=Capsule, length=9.9077, radius=6.3987, rotation=<rot 90 0 0>}]
 ponytail_A [{shapeType=Capsule, length=4.75, radius=3.25, center=<1.9794 0.0001 0>, rotation=<rot 90 0 0>}]
 ponytail_B [{shapeType=Capsule, length=4.75, radius=3.25, center=<1.9794 0.0001 0>, rotation=<rot 90 0 0>}]
}
constraints [
 {bone1Name=ponytail_A, bone2Name=head, swing1LimitDegrees=40, swing2LimitDegrees=40, twistLimitDegrees=10}
 {bone1Name=ponytail_B, bone2Name=head, swing1LimitDegrees=40, swing2LimitDegrees=40, twistLimitDegrees=10}
 {bone1Name=dyn_ribbon_l, bone2Name=head, swing1LimitDegrees=20, swing2LimitDegrees=30, twistLimitDegrees=45, twistMotion=Locked}
 {bone1Name=dyn_ribbon_r, bone2Name=head, swing1LimitDegrees=20, swing2LimitDegrees=30, twistLimitDegrees=45, twistMotion=Locked}
]
